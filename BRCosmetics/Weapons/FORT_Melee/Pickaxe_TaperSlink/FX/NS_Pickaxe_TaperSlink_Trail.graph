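# Niagara swing-trail effect for the TaperSlink pickaxe, drawing green ribbon trails with extra wing ribbons.
# InnerTrails and OuterTrails spawn anim trails, each feeding three ribbon renderers layered with separately computed colors and widths for glow, laser, plasma and detail looks.
# Wings builds three trails from socket pairs 2-3, 4-5 and 6-7 into one renderer.
# All emitters use system TODColor for tint and SwingGeneralFade to fade with the swing.
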NiagaraSystem /BRCosmetics/Weapons/FORT_Melee/Pickaxe_TaperSlink/FX/NS_Pickaxe_TaperSlink_Trail
dependencies {
 usedDataInterfaces [&NiagaraDataInterfaceCurve, &NiagaraDataInterfaceSocketReader, &NiagaraDataInterfaceColorCurve]
 usedDynamicInputs [&ColorFromCurve]
 usedModules [
  &SystemState
  &self:SystemSpawnScript.SystemScriptSource.SystemScriptGraph.NiagaraNodeAssignment_1.SetVariables_#1
  &self:SystemSpawnScript.SystemScriptSource.SystemScriptGraph.NiagaraNodeAssignment_5.SetVariables_#2
  &self:SystemSpawnScript.SystemScriptSource.SystemScriptGraph.NiagaraNodeAssignment_0.SetVariables_#3
  &self:SystemSpawnScript.SystemScriptSource.SystemScriptGraph.NiagaraNodeAssignment_3.SetVariables_#4
  &self:SystemSpawnScript.SystemScriptSource.SystemScriptGraph.NiagaraNodeAssignment_4.SetVariables_#5
 ]
}
emitters {
 InnerTrails {
  emitterSpawnScript [&FN_RibbonIDOffsetUtility()]
  emitterUpdateScript [
   &EmitterState()
   &FN_SpawnAnimTrail()
  ]
  particleSpawnScript [
   &FN_InitializeAnimTrailBoneIndecies(Second Bone or Socket Index=0)
   &FN_InitializeAnimTrail(AnimTrailWidth=20, Color=<rgba 0.7895 3.1 0.1734 0.8>, Lifetime=0.2)
  ]
  particleUpdateScript [
   &FN_UpdateAnimTrail() {module=FN_UpdateAnimTrail001}
   &ScaleColor(
    Scale Alpha=&Multiply_Float(A=$System.SwingGeneralFade, B=1)
    Scale RGB=&MakeVectorFromLinearColorRGB(Color=$System.TODColor)
    ScaleA=true
    ScaleRGB=true
   )
   /SetVariables(Particles.GlowColor=&MultiplyLinearColor(
     Linear Color=&MakeLinearColorFromVectorAndFloat(
      "Vector (RGB)"=&MakeVectorFromLinearColorRGB(Color=<rgba 0.4839 1 0.2717 1>)
      "Float (Alpha)"=&FloatFromCurve(FloatCurve=DataInterface() {propertyValues {Curve {keys [<at 0 1 RCIM_Cubic>,<at 0.5 0 RCIM_Cubic>]}}})
     )
     Scale Factor=&MakeVector4_FromLinearColor(LinearColor=$System.TODColor)
    )) {isSetParameters=true}
   /SetVariables(Particles.LaserColor:LinearColor=&MultiplyLinearColor(
     Linear Color=&MakeLinearColorFromVectorAndFloat(
      "Vector (RGB)"=&MakeVectorFromLinearColorRGB(Color=<rgba 3.8629 10 1.6487 1>)
      "Float (Alpha)"=$System.SwingGeneralFade
     )
     Scale Factor=&MakeVector4_FromLinearColor(LinearColor=$System.TODColor)
    )) {isSetParameters=true}
   &ScaleRibbonWidth(Ribbon Width Scale=&Multiply_Float(
     A=&FloatFromCurve(FloatCurve=DataInterface() {propertyValues {Curve {keys [<at 0.075 1 RCIM_Cubic>, {arriveTangentWeight=0.1917, interpMode=RCIM_Cubic, tangentMode=RCTM_User, time=0.35, value=0}]}}})
     B=$System.SwingGeneralFade
    ))
   /SetVariables(Particles.RibbonWidthAlt:NiagaraFloat=&Multiply_Float(A=$Particles.RibbonWidth, B=1.8)) {isSetParameters=true}
   /SetVariables(Particles.RibbonWidth_Laser:NiagaraFloat=&Multiply_Float(
     A=&FloatFromCurve(
      FloatCurve=DataInterface() {propertyValues {Curve {keys [
          <at 0.075 1 RCIM_Cubic>
          {arriveTangent=-4.8146, arriveTangentWeight=0.1917, interpMode=RCIM_Cubic, leaveTangent=-4.8146, tangentMode=RCTM_User, time=0.3, value=0}
         ]}}}
      Scale Curve=20
     )
     B=1
    )) {isSetParameters=true}
  ]
  rendererClasses [&NiagaraRibbonRendererProperties]
  renderers [
   {rendererClass=&NiagaraRibbonRendererProperties}
   {rendererClass=&NiagaraRibbonRendererProperties, rendererIndex=1}
   {rendererClass=&NiagaraRibbonRendererProperties, rendererIndex=2}
  ]
 }
 OuterTrails {
  emitterUpdateScript [
   &EmitterState()
   &FN_SpawnAnimTrail()
  ]
  particleSpawnScript [
   &FN_InitializeAnimTrailBoneIndecies(Second Bone or Socket Index=0)
   &FN_InitializeAnimTrail(AnimTrailWidth=22, Color=<rgba 0.0532 1 0 0.8>, Lifetime=0.16)
  ]
  particleUpdateScript [
   &FN_UpdateAnimTrail() {module=FN_UpdateAnimTrail001}
   &ScaleColor(
    Scale Alpha=&Multiply_Float(
     A=$System.SwingGeneralFade
     B=&FloatFromCurve(
      FloatCurve=DataInterface() {propertyValues {Curve {
         defaultValue=1
         keys [{interpMode=RCIM_Cubic, tangentMode=RCTM_User, time=0, value=1}, {interpMode=RCIM_Cubic, tangentMode=RCTM_User, time=1, value=0}]
        }}}
      Scale Curve=0.2
     )
    )
    Scale RGB=&MakeVectorFromLinearColorRGB(Color=$System.TODColor)
    ScaleA=true
    ScaleRGB=true
   )
   &ScaleRibbonWidth(Ribbon Width Scale=&FloatFromCurve(FloatCurve=DataInterface() {propertyValues {Curve {
        defaultValue=1
        keys [<at 0 1 RCIM_Cubic>, {arriveTangent=-0.6833, interpMode=RCIM_Cubic, leaveTangent=-0.6833, tangentMode=RCTM_User, time=1, value=0}]
       }}}))
   /SetVariables(
    Particles.PlasmaTrail_Color:LinearColor=&MultiplyLinearColor(
     Linear Color=<rgba 0.0532 1 0 0.8>
     Scale Factor=&MakeVector4FromVectorAndFloat(
      "Vector (RGB)"=&MakeVectorFromLinearColorRGB(Color=$System.TODColor)
      "Float (Alpha)"=&Multiply_Float(A=$System.SwingGeneralFade, B=0.2)
     )
    )
    Particles.PlasmaTrail_Width:NiagaraFloat=&Multiply_Float(
     A=17
     B=&FloatFromCurve(
      FloatCurve=DataInterface() {propertyValues {Curve {
         defaultValue=1
         keys [
          {interpMode=RCIM_Cubic, tangentMode=RCTM_User, time=0, value=1}
          {arriveTangent=-0.5791, interpMode=RCIM_Cubic, leaveTangent=-0.5791, tangentMode=RCTM_User, time=0.5, value=0}
         ]
        }}}
      Scale Curve=1.1
     )
    )
   ) {isSetParameters=true}
   /SetVariables(
    Particles.DetailTrail_Color:LinearColor=&MultiplyLinearColor(
     Linear Color=<rgba 0.6747 2.1 0.2662 2>
     Scale Factor=/Niagara/DynamicInputs/Multiply/Multiply_Vector4_ByFloat(
      Vector4:Vector4f=&MakeVector4_FromLinearColor(LinearColor=$System.TODColor)
      Float:NiagaraFloat=$System.SwingGeneralFade
     )
    )
    Particles.DetailTrail_Width:NiagaraFloat=&Multiply_Float(
     A=25
     B=&FloatFromCurve(
      FloatCurve=DataInterface() {propertyValues {Curve {keys [<at 0.075 1 RCIM_Cubic>, {arriveTangentWeight=0.1917, interpMode=RCIM_Cubic, tangentMode=RCTM_User, time=0.4, value=0}]}}}
      Scale Curve=&Multiply_Float(A=0.6, B=$System.SwingGeneralFade)
     )
    )
   ) {isSetParameters=true}
  ]
  rendererClasses [&NiagaraRibbonRendererProperties]
  renderers [
   {rendererClass=&NiagaraRibbonRendererProperties}
   {rendererClass=&NiagaraRibbonRendererProperties, rendererIndex=1}
   {rendererClass=&NiagaraRibbonRendererProperties, rendererIndex=2}
  ]
 }
 Wings {
  emitterUpdateScript [
   &EmitterState(Life Cycle Mode=::NewEnumerator1, Loop Behavior=::NewEnumerator1, Loop Delay=0.05, Loop Duration=0.25, UseLoopDelay=false)
   &FN_SpawnAnimTrail(Number of Anim Trails=3, Spawn Interval=0)
  ]
  particleSpawnScript [
   &FN_InitializeAnimTrailBoneIndecies(First Bone or Socket Index=2, Second Bone or Socket Index=3)
   &FN_InitializeAnimTrailBoneIndecies(AnimTrail Selection=1, First Bone or Socket Index=4, Second Bone or Socket Index=5) {module=FN_InitializeAnimTrailBoneIndecies001}
   &FN_InitializeAnimTrailBoneIndecies(AnimTrail Selection=2, First Bone or Socket Index=6, Second Bone or Socket Index=7) {module=FN_InitializeAnimTrailBoneIndecies002}
   &FN_InitializeAnimTrail(Color=$System.TODColor, Lifetime=0.16)
  ]
  particleUpdateScript [
   &FN_UpdateAnimTrail() {module=FN_UpdateAnimTrail001}
   &ScaleColor(
    Scale Alpha=&Multiply_Float(
     A=&FloatFromCurve(
      FloatCurve=DataInterface() {propertyValues {Curve {
         defaultValue=0.5
         keys [
          {arriveTangent=0.1062, interpMode=RCIM_Cubic, leaveTangent=0.1062, tangentMode=RCTM_User, time=0, value=1}
          {arriveTangent=-2.6567, interpMode=RCIM_Cubic, leaveTangent=-2.6567, tangentMode=RCTM_User, time=1, value=0}
         ]
        }}}
      Scale Curve=0.5
     )
     B=$System.SwingGeneralFade
    )
    Scale RGB=&MakeVectorFromLinearColorRGB(Color=<rgba 5 5 5 0>)
    ScaleA=true
    ScaleRGB=true
   )
  ]
  rendererClasses [&NiagaraRibbonRendererProperties]
  renderers [{rendererClass=&NiagaraRibbonRendererProperties}]
 }
}
summary {
 systemName=NS_Pickaxe_TaperSlink_Trail
 emitters [
  {emitterName=InnerTrails, rendererClasses [&NiagaraRibbonRendererProperties]}
  {emitterName=Wings, rendererClasses [&NiagaraRibbonRendererProperties]}
  {emitterName=OuterTrails, rendererClasses [&NiagaraRibbonRendererProperties]}
 ]
 userVariables [
  User.bIsFrontEnd {type=NiagaraBool}
  User.bIsFrontEndPreview {type=NiagaraBool}
  User.RibbonIDOffset {type=NiagaraInt32}
 ]
}
